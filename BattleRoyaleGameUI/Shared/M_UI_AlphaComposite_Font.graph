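# UI-domain material that, by its name, tints font text using AlphaComposite (premultiplied alpha) blending.
# The Color parameter's RGB is scaled by one minus Dim, then multiplied by vertex alpha to give a premultiplied color.
# A separate branch inverts AdditiveMixRatio, a value whose name points to a mix between additive and standard blending.
# The graph shows no texture samples and no links to material outputs.
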Material /BattleRoyaleGameUI/Shared/M_UI_AlphaComposite_Font
graph {
 properties {bAutomaticallySetUsageInEditor=false, blendMode=BLEND_AlphaComposite, materialDomain=MD_UI, usedShadingModels=Default Lit}
 nodes [
  n1 = ScalarParameter() {defaultValue=0.25, parameterName=Dim}
  Multiply(A=VertexColor().A, B=Multiply(A=VectorParameter() {parameterName=Color, defaultValue=<rgba 0.25 0.25 0.25 1>}.RGB, B=OneMinus(None=@n1)))
  OneMinus(None=ScalarParameter() {defaultValue=0.75, parameterName=AdditiveMixRatio})
 ]
}
