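# UI material function in BattleRoyaleGameUI that swaps a value for a cheaper-tier alternative based on material quality level.
# A Quality Switch routes the LowEnd input to the Low, Medium and High levels, while any other level, such as Epic, receives the Default input.
# Both inputs use a constant preview as their default, so either can be left unconnected.
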MaterialFunction /BattleRoyaleGameUI/Shared/MF_UI_LowEnd_ValueSwitcher
graph {
 properties {previewMaterialDomain=MD_UI}
 nodes [
  n1 = Constant() {r=0.5}
  n2 = FunctionInput(Preview=Constant() {r=0.5}) {bUsePreviewValueAsDefault=true, inputName=LowEnd, sortPriority=1}
  FunctionOutput(None=QualitySwitch(Default=FunctionInput(Preview=@n1) {bUsePreviewValueAsDefault=true, inputName=Default}, Low=@n2, High=@n2, Medium=@n2))
 ]
}
