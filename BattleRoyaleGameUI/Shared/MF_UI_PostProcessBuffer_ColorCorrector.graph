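# UI material function that color-corrects a source buffer, named for use on the UI post-process buffer.
# It passes the buffer with brightness, contrast, saturation, hue shift, lift/gamma/gain and effect amount inputs into the shared MF_UI_ColorCorrector, then saturates the result.
# Inputs use their preview values as defaults, giving a preset grade with per-channel lift, gamma and gain.
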MaterialFunction /BattleRoyaleGameUI/Shared/MF_UI_PostProcessBuffer_ColorCorrector
graph {
 properties {previewMaterialDomain=MD_UI}
 nodes [
  n1 = NamedRerouteUsage()
  n2 = FunctionInput(Preview=Constant() {r=1.5}) {bUsePreviewValueAsDefault=true, inputName=Brightness, inputType=FunctionInput_Scalar, sortPriority=5}
  n3 = FunctionInput(Preview=Constant() {r=0.5}) {bUsePreviewValueAsDefault=true, inputName=Contrast, inputType=FunctionInput_Scalar, sortPriority=7}
  n4 = FunctionInput(Preview=Constant() {r=0.5}) {bUsePreviewValueAsDefault=true, inputName=Saturation, inputType=FunctionInput_Scalar, sortPriority=4}
  n5 = FunctionInput(Preview=Constant()) {bUsePreviewValueAsDefault=true, inputName=HueShift, inputType=FunctionInput_Scalar, sortPriority=6}
  n6 = FunctionInput(Preview=Constant4Vector() {constant=<rgba 0.0015 0.003 0.002 1>}.RGB) {bUsePreviewValueAsDefault=true, inputName=Lift}
  n7 = FunctionInput(Preview=Constant4Vector() {constant=<rgba 1.55 1.45 1.666 1>}.RGB) {bUsePreviewValueAsDefault=true, inputName=Gamma, sortPriority=1}
  n8 = FunctionInput(Preview=Constant4Vector() {constant=<rgba 2 2 2.5 1>}.RGB) {bUsePreviewValueAsDefault=true, inputName=Gain, sortPriority=2}
  n9 = FunctionInput(Preview=Constant() {r=100}) {bUsePreviewValueAsDefault=true, inputName=EffectPercent, inputType=FunctionInput_Scalar, sortPriority=8}
  FunctionOutput(None=Saturate(None=MaterialFunctionCall!(Color=@n1, Brightness=@n2, Contrast=@n3, Saturation=@n4, Hue Shift=@n5, Lift=@n6, Gamma=@n7, Gain=@n8, Effect Amount=@n9) {materialFunction=&/MF_UI_ColorCorrector}.Result))
  NamedRerouteDeclaration(None=FunctionInput() {bUsePreviewValueAsDefault=true, inputName=SourceBuffer}) {name=BufferRGB, nodeColor=<rgba 0 0 0 1>}
 ]
}
