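# UI material function that derives a palette of adaptive colors from a blurred or captured Slate post buffer for Battle Royale game UI.
# The BufferSource input, defaulting to Slate post buffer 0, feeds separate color-corrector calls producing Dark, Dim, Medium and Bright outputs.
# Low-end value switchers substitute flat grays for each tone and a Source parameter for the pass-through SourceOutput.
# White comes directly from the MPC_BRGameUI parameter collection.
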MaterialFunction /BattleRoyaleGameUI/Shared/MF_UI_AdaptiveColors
graph {
 properties {previewMaterialDomain=MD_UI}
 nodes [
  n1 = NamedRerouteUsage()
  n2 = Constant() {r=0.125}
  n3 = Constant() {r=1.125}
  n4 = Constant()
  n5 = Constant() {r=1}
  n6 = Constant() {r=100}
  FunctionOutput(None=MaterialFunctionCall!(
    Default=MaterialFunctionCall!(Lift, SourceBuffer=@n1, Gamma, Gain, Saturation=@n2, Brightness=@n3, HueShift=@n4, Contrast=@n5, EffectPercent=@n6) {materialFunction=&/MF_UI_PostProcessBuffer_ColorCorrector}.Result
    LowEnd=Constant4Vector() {constant=<rgba 0.5 0.5 0.5 1>}.RGB
   ) {materialFunction=&/MF_UI_LowEnd_ValueSwitcher}.Result) {outputName=Dim, sortPriority=2}
  n7 = NamedRerouteUsage()
  n8 = Constant() {r=0.75}
  n9 = Constant() {r=0.07}
  n10 = Constant()
  n11 = Constant() {r=0.9}
  n12 = Constant() {r=100}
  FunctionOutput(None=MaterialFunctionCall(
    Default=MaterialFunctionCall(SourceBuffer=@n7, Saturation=@n8, Brightness=@n9, HueShift=@n10, Contrast=@n11, EffectPercent=@n12) {materialFunction=&MF_UI_PostProcessBuffer_ColorCorrector}.Result
    LowEnd=Constant4Vector() {constant=<rgba 0.013 0.013 0.013 1>}.RGB
   ) {materialFunction=&MF_UI_LowEnd_ValueSwitcher}.Result) {outputName=Dark}
  n13 = NamedRerouteUsage()
  n14 = Constant()
  n15 = Constant() {r=1.25}
  n16 = Constant()
  n17 = Constant() {r=0.5}
  n18 = Constant() {r=100}
  FunctionOutput(None=MaterialFunctionCall(
    Default=MaterialFunctionCall(SourceBuffer=@n13, Saturation=@n14, Brightness=@n15, HueShift=@n16, Contrast=@n17, EffectPercent=@n18) {materialFunction=&MF_UI_PostProcessBuffer_ColorCorrector}.Result
    LowEnd=Constant4Vector() {constant=<rgba 0.7 0.7 0.7 1>}.RGB
   ) {materialFunction=&MF_UI_LowEnd_ValueSwitcher}.Result) {outputName=Medium, sortPriority=3}
  n19 = NamedRerouteUsage()
  n20 = Constant()
  n21 = Constant() {r=2.5}
  n22 = Constant() {r=-0.02}
  n23 = Constant() {r=0.75}
  n24 = Constant() {r=100}
  FunctionOutput(None=MaterialFunctionCall(
    Default=MaterialFunctionCall(SourceBuffer=@n19, Saturation=@n20, Brightness=@n21, HueShift=@n22, Contrast=@n23, EffectPercent=@n24) {materialFunction=&MF_UI_PostProcessBuffer_ColorCorrector}.Result
    LowEnd=Constant4Vector() {constant=<rgba 0.825 0.825 0.825 1>}.RGB
   ) {materialFunction=&MF_UI_LowEnd_ValueSwitcher}.Result) {outputName=Bright, sortPriority=4}
  FunctionOutput(None=ComponentMask(None=CollectionParameter() {collection=&/MPC_BRGameUI, parameterName=White}) {b=true, g=true, r=true}) {outputName=White, sortPriority=5}
  FunctionOutput(None=MaterialFunctionCall(Default=NamedRerouteUsage(), LowEnd=VectorParameter() {parameterName=Source, defaultValue=<rgba 0.212 0.212 0.212 1>}.RGB) {materialFunction=&MF_UI_LowEnd_ValueSwitcher}.Result) {outputName=SourceOutput}
  MaterialFunctionCall!("UVs (Optional)") {materialFunction=/Engine/Functions/UserInterface/GetSlatePost1}
  NamedRerouteDeclaration(None=FunctionInput(Preview=MaterialFunctionCall!("UVs (Optional)") {materialFunction=/Engine/Functions/UserInterface/GetSlatePost0}.LinearRGB) {bUsePreviewValueAsDefault=true, inputName=BufferSource}) {name=SourceBuffer, nodeColor=<rgba 0 0 0 1>}
 ]
}
