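# Translucent UI font material that renders a soft shadow or glow from the font's signed distance field.
# A falloff index blends between ease curves, then anti-aliasing with threshold and smoothing shapes the alpha, scaled by animated opacity, vertical fade and hide controls.
# Color comes from custom, adaptive or Red/Cyan critical palettes, with a static switch premultiplying it for alpha-composite blending with an additive mix.
# Move_X and Move_Y offset the position.
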Material /BattleRoyaleGameUI/Shared/M_UI_Font_SoftShadow
graph {
 properties {bAutomaticallySetUsageInEditor=false, blendMode=BLEND_Translucent, materialDomain=MD_UI, usedShadingModels=Default Lit}
 nodes [
  MaterialExpressionWorldPosition_0 = WorldPosition()
  n1 = Convert(None=@MaterialExpressionWorldPosition_0.XY) {
   nodeName=Break Float2
   convertInputs [{
     type=Vector2
     expressionInput {
      expression=&self:MaterialExpressionWorldPosition_0
      inputName=None
      mask=1
      maskA=0
      maskB=0
      maskG=1
      maskR=1
      outputIndex=1
     }
    }]
   convertOutputs [{type=Scalar}, {type=Scalar}]
  }
  AppendVector(A=Add(A=@n1, B=ScalarParameter() {parameterName=Move_X}), B=Add(A=@n1, B=ScalarParameter() {parameterName=Move_Y}))
  n2 = FontSignedDistance()
  n3 = MaterialFunctionCall("Value (0-1)"=@n2.Signed Distance) {materialFunction=&MF_UI_EaseCurves}
  n4 = MaterialFunctionCall("Value (0-1)"=@n3.EaseInQuint) {materialFunction=&MF_UI_EaseCurves}
  n5 = ScalarParameter() {defaultValue=40, parameterName=ExtremeCurveMultiplier}
  NamedRerouteDeclaration(None=Saturate(None=Multiply(A=@n4.EaseInSine, B=@n5))) {name=ExtremeCurve, nodeColor=<rgba 0 0 0 1>}
  n6 = ScalarParameter() {parameterName=AdditiveMixRatio}
  n7 = ScalarParameter() {defaultValue=3, parameterName=FalloffRate_Index}
  n8 = Saturate(None=@n7)
  n9 = Subtract(A=@n7)
  n10 = LinearInterpolate(
   A=LinearInterpolate(A=LinearInterpolate(A=@n2.Signed Distance, B=@n3.EaseInSine, Alpha=@n8), B=@n3.EaseInCubic, Alpha=Saturate(None=@n9))
   B=@n3.EaseInQuint
   Alpha=Saturate(None=Subtract(A=@n7) {constB=2})
  )
  n11 = ScalarParameter() {defaultValue=0.5, parameterName=Threshold}
  n12 = ScalarParameter() {defaultValue=1, parameterName=Smoothing}
  n13 = ScalarParameter() {defaultValue=0.5, parameterName=Opacity}
  n14 = ScalarParameter() {defaultValue=1, parameterName=AnimOpacity}
  n15 = Multiply(
   A=Saturate(None=MaterialFunctionCall(DistanceField=@n10, Threshold=@n11, Smoothing=@n12) {materialFunction=&MF_UI_AntiAliasing}.Outer)
   B=NamedRerouteDeclaration(None=LinearInterpolate(B=@n13, Alpha=@n14)) {name=Opacity, nodeColor=<rgba 0 0 0 1>}
  )
  n16 = ScalarParameter() {parameterName=FadeOpacity}
  n17 = OneMinus(None=ComponentMask(None=TextureCoordinate()) {g=true})
  n18 = NamedRerouteDeclaration(None=Saturate(None=MaterialFunctionCall("Value (0-1)"=@n17) {materialFunction=&MF_UI_EaseCurves}.EaseInCubic)) {name=FadeMask, nodeColor=<rgba 0 0 0 1>}
  n19 = LinearInterpolate(A=@n15, B=Multiply(A=@n15, B=LinearInterpolate(A=@n16, Alpha=@n18)), Alpha=ScalarParameter() {parameterName=UseFade})
  n20 = LinearInterpolate(A=@n19, Alpha=ScalarParameter() {parameterName=Hide}) {constB=0}
  StaticSwitchParameter(True=Multiply(A=OneMinus(None=@n6), B=@n20), False=@n20) {parameterName=IsAlphaCompositeBlendMode}
  n21 = VectorParameter() {parameterName=Color, defaultValue=<rgba 1 0 0 1>}
  n22 = LinearInterpolate(A=VectorParameter() {parameterName=Color_Falloff, defaultValue=<rgba 1 0 0 1>}.RGB, B=@n21.RGB, Alpha=NamedRerouteUsage())
  n23 = ScalarParameter() {parameterName=UseFade}
  n24 = MaterialFunctionCall!(BufferSource) {materialFunction=&/MF_UI_AdaptiveColors}
  n25 = ComponentMask(None=CollectionParameter() {collection=&/MPC_BRGameUI, parameterName=Red}) {b=true, g=true, r=true}
  n26 = ComponentMask(None=CollectionParameter() {collection=&MPC_BRGameUI, parameterName=Cyan}) {b=true, g=true, r=true}
  n27 = ScalarParameter() {parameterName=PickCriticalColor}
  n28 = LinearInterpolate(
   A=StaticSwitchParameter(True=LinearInterpolate(A=@n21.RGB, B=@n22, Alpha=@n23), False=@n24.Dark) {parameterName=CustomColor}
   B=LinearInterpolate(A=LinearInterpolate(A=@n25, B=@n26, Alpha=@n27), B=@n24.Dark, Alpha=NamedRerouteUsage()) {constB=0}
   Alpha=ScalarParameter() {parameterName=IsCriticalColor}
  )
  StaticSwitchParameter(True=Multiply(A=VertexColor().A, B=Multiply(A=@n28, B=@n19)), False=@n28) {parameterName=IsAlphaCompositeBlendMode}
 ]
}
