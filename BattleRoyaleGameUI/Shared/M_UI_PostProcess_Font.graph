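# UI-domain translucent material for fonts in the Battle Royale game UI, tinting them with adaptive colors.
# MF_UI_AdaptiveColors supplies source, Dim, Medium, Bright and White outputs.
# Chained lerps, driven by saturated offsets of the AdaptiveColorIndex scalar, step or blend through those tones without branching.
# An Opacity parameter passes through MF_UI_LowEnd_ValueSwitcher, which receives a slightly reduced value for low-end.
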Material /BattleRoyaleGameUI/Shared/M_UI_PostProcess_Font
graph {
 properties {bAutomaticallySetUsageInEditor=false, blendMode=BLEND_Translucent, materialDomain=MD_UI, usedShadingModels=Default Lit}
 nodes [
  n1 = MaterialFunctionCall!(BufferSource) {materialFunction=&/MF_UI_AdaptiveColors}
  n2 = ScalarParameter() {parameterName=AdaptiveColorIndex}
  n3 = LinearInterpolate(
   A=LinearInterpolate(A=@n1.SourceOutput, B=@n1.SourceOutput, Alpha=ScalarParameter() {parameterName=AdaptiveColorIndex})
   B=@n1.Dim
   Alpha=Saturate(None=Subtract(A=@n2))
  )
  n4 = Saturate(None=Subtract(A=ScalarParameter() {parameterName=AdaptiveColorIndex}) {constB=2})
  n5 = Subtract(A=ScalarParameter() {parameterName=AdaptiveColorIndex}) {constB=3}
  n6 = ScalarParameter() {parameterName=AdaptiveColorIndex}
  LinearInterpolate(
   A=LinearInterpolate(A=LinearInterpolate(A=@n3, B=@n1.Medium, Alpha=@n4), B=@n1.Bright, Alpha=Saturate(None=@n5))
   B=@n1.White
   Alpha=Saturate(None=Subtract(A=@n6) {constB=4})
  )
  n7 = ScalarParameter() {defaultValue=1, parameterName=Opacity}
  MaterialFunctionCall!(Default=@n7, LowEnd=Multiply(A=@n7) {constB=0.9}) {materialFunction=&/MF_UI_LowEnd_ValueSwitcher}
 ]
}
