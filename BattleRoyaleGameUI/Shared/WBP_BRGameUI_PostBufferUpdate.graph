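# Battle royale UI container widget that shows or removes a post-buffer update child widget on demand.
# HandleShowPostBuffer, when passed true, asynchronously creates WBP_BRGameUI_PostBufferChild from a soft class reference and adds it to the root overlay, filling it.
# When passed false, it clears the overlay's children; an async creation still in flight is not cancelled.
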WidgetBlueprint /BattleRoyaleGameUI/Shared/WBP_BRGameUI_PostBufferUpdate
graphs [EventGraph {nodes [
   n1 = HandleShowPostBuffer!() -> (OutputDelegate, then, bShowPostBuffer:bool=~) {is=title K2Node_CustomEvent}
   > n2 = Branch(Condition=@n1.bShowPostBuffer)
   > n3 = CreateWidgetAsync!(
    execute:Exec
    UserWidgetSoftClass:User Widget Soft Class Reference=&/WBP_BRGameUI_PostBufferChild.WBP_BRGameUI_PostBufferChild_C
    OwningPlayer:Player Controller Object Reference=GetOwningPlayer()
    bSuspendInputUntilComplete:bool=false
   ) -> (then:Exec, OnComplete:Exec, OnFailure:Exec, UserWidget:User Widget Object Reference) {is=title K2Node_AsyncAction}
   n4 = AddChildToOverlay!(
    execute=@n3.OnComplete
    self:Overlay Object Reference=Get Overlay_PostBufferUpdate() -> (Overlay_PostBufferUpdate:Overlay Object Reference)
    Content:Widget Object Reference=@n3
   ) -> (then, ReturnValue:Overlay Slot Object Reference)
   > SetHorizontalAlignment(self=@n4, InHorizontalAlignment=HAlign_Fill)
   > SetVerticalAlignment(self=@n4, InVerticalAlignment=VAlign_Fill)
   ClearChildren(execute=@n2.else, self=Get Overlay_PostBufferUpdate())
  ]}]
variables []
widgets {
 info {inheritedWidgetCount=0, parentClass=&UserWidget, rootWidgetClass=&Overlay, widgetCount=1}
 widgets [Overlay_PostBufferUpdate {widgetClassPath=&Overlay, bIsVariable=true, widget=Overlay_PostBufferUpdate}]
}
