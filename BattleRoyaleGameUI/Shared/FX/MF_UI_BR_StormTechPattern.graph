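# UI material function that generates an animated, storm-themed tech pattern of flickering rounded tiles and outlines for Battle Royale UI.
# Two screen-space particle grids, one at double DigiTiling, feed time-seeded random values that drive box SDF sizes and visibility thresholds.
# Small filled tiles are added to larger anti-aliased tiles blended with outlines, each layer having its own opacity parameter.
# FXIntensity scales the saturated mask output, and TimeOffset shifts the animation.
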MaterialFunction /BattleRoyaleGameUI/Shared/FX/MF_UI_BR_StormTechPattern
graph {nodes [
  n1 = MaterialFunctionCall!(ScreenUV, TextureSize, ScalePivotPoint) {materialFunction=/FortUILibrary/MaterialFunctions/Basics/MF_UI_ScreenSpaceUVScaling}
  n2 = Constant()
  n3 = MaterialFunctionCall!(
   UV=@n1.Scaled to Fit
   Global Offset
   Global Rotation
   Tiling=Multiply(A=ScalarParameter() {defaultValue=36, parameterName=DigiTiling}) {constB=2}
   Cell Offset=@n2
   Cell Scale=Constant() {r=1}
   Cell Rotation=@n2
   Random Seed
  ) {materialFunction=/FortUILibrary/MaterialFunctions/Basics/MF_UI_ParticleGrid}
  n4 = NamedRerouteUsage()
  n5 = MaterialFunctionCall!(UV=@n3.Random Cell Values, Seed=Multiply(A=@n4) {constB=0.333}) {materialFunction=&MF_UI_Random}
  n6 = MaterialFunctionCall(UVs=@n3.Cells UV, Size=@n5.Result, CornerRoundness=Constant() {r=0.1}) {materialFunction=&MF_UI_Box}
  n7 = Saturate(None=Subtract(A=@n5.Result) {constB=0.75})
  n8 = Constant()
  n9 = MaterialFunctionCall(
   UV=@n1.Scaled to Fit
   Tiling=ScalarParameter() {defaultValue=36, parameterName=DigiTiling}
   Cell Offset=@n8
   Cell Scale=Constant() {r=1}
   Cell Rotation=@n8
  ) {materialFunction=&MF_UI_ParticleGrid}
  n10 = NamedRerouteUsage()
  n11 = MaterialFunctionCall(UV=@n9.Random Cell Values, Seed=Multiply(A=@n10) {constB=0.25}) {materialFunction=&MF_UI_Random}
  n12 = MaterialFunctionCall(UVs=@n9.Cells UV, Size=@n11.Result, CornerRoundness=Constant() {r=0.2}) {materialFunction=&MF_UI_Box}
  n13 = Saturate(None=Subtract(A=@n11.Result) {constB=0.25})
  n14 = Multiply(
   A=Min(
    A=MaterialFunctionCall(DistanceField=@n12.SDF) {materialFunction=&MF_UI_AntiAliasing}.Inner
    B=@n13
   )
   B=ScalarParameter() {defaultValue=0.125, parameterName=TileLayer2_Opacity}
  ) {constB=0.125}
  n15 = Saturate(None=Subtract(A=@n11.Result) {constB=0.5})
  n16 = MaterialFunctionCall(
   UVs=@n9.Cells UV
   Size=MaterialFunctionCall(UV=@n9.Random Cell Values, Seed=Multiply(A=@n10) {constB=0.35}) {materialFunction=&MF_UI_Random}.Result
   CornerRoundness=Constant() {r=0.2}
   OutlineThickness=Constant() {r=0.05}
  ) {materialFunction=&MF_UI_Box}
  n17 = Multiply(
   A=Min(
    A=OneMinus(None=@n15)
    B=MaterialFunctionCall(DistanceField=@n16.SDFOutline) {materialFunction=&MF_UI_AntiAliasing}.Inner
   )
   B=ScalarParameter() {defaultValue=0.5, parameterName=LinesOpacity}
  ) {constB=0.25}
  n18 = Add(
   A=Multiply(A=Min(A=@n6.Fill, B=@n7), B=ScalarParameter() {defaultValue=0.5, parameterName=TileLayer1_Opacity}) {constB=0.5}
   B=LinearInterpolate(A=Multiply(A=@n14, B=@n13), Alpha=Multiply(A=@n15, B=@n17))
  )
  n19 = ScalarParameter() {defaultValue=3, parameterName=FXIntensity}
  FunctionOutput(None=Saturate(None=Multiply(A=@n18, B=@n19)))
  Multiply(A=@n4) {constB=0.45}
  NamedRerouteDeclaration(None=Add(A=Time(), B=ScalarParameter() {parameterName=TimeOffset}) {constB=0.2}) {name=TimeValue, nodeColor=<rgba 0 0 0 1>}
  NamedRerouteUsage()
 ]}
