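# UI material with alpha-composite blending that draws a soft glow, named for elimination effects in the Battle Royale UI.
# A 9-slice circle SDF shapes the glow, multiplied by an eased horizontal sine gradient.
# Color blends dimmed shared Red toward StormPink from the UI parameter collection, with an optional custom color.
# AdditiveRatio reduces alpha for additive-style glow, while Anim_Opacity fades color and alpha.
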Material /BattleRoyaleGameUI/Shared/FX/M_UI_ElimGlow
graph {
 properties {bAutomaticallySetUsageInEditor=false, blendMode=BLEND_AlphaComposite, materialDomain=MD_UI, usedShadingModels=Default Lit}
 nodes [
  n1 = NamedRerouteUsage()
  n2 = OneMinus(None=MaterialFunctionCall(UVs=MaterialFunctionCall!(UVs) {materialFunction=/FortUILibrary/MaterialFunctions/Basics/MF_UI_9SliceUVs}.UVs) {materialFunction=&MF_UI_Circle}.SDF)
  n3 = Saturate(None=MaterialFunctionCall("Value (0-1)"=SmoothStep(Value=@n2) {constMax=1.5, constMin=0.5}) {materialFunction=&MF_UI_EaseCurves}.EaseInCubic)
  n4 = ScalarParameter() {defaultValue=1, parameterName=AdditiveRatio}
  LinearInterpolate(B=Multiply(A=Multiply(A=@n1, B=@n3), B=OneMinus(None=@n4)), Alpha=ScalarParameter() {defaultValue=1, parameterName=Anim_Opacity})
  n5 = Sine(None=MaterialFunctionCall() {materialFunction=&GetUserInterfaceUV}.Normalized UV) {period=2}
  n6 = NamedRerouteDeclaration(None=MaterialFunctionCall("Value (0-1)"=ComponentMask(None=@n5) {r=true}) {materialFunction=&MF_UI_EaseCurves}.EaseInQuint) {name=Gradient, nodeColor=<rgba 0 0 0 1>}
  n7 = CollectionParameter() {collection=/BattleRoyaleGameUI/Shared/MPC_BRGameUI, parameterName=Red}
  n8 = LinearInterpolate(
   A=Multiply(A=ComponentMask(None=@n7) {b=true, g=true, r=true}, B=ScalarParameter() {defaultValue=0.25, parameterName=DimColor})
   B=ComponentMask(None=CollectionParameter() {collection=&MPC_BRGameUI, parameterName=StormPink}) {b=true, g=true, r=true}
   Alpha=ScalarParameter() {parameterName=StromPink}
  )
  n9 = VectorParameter() {parameterName=CustomColor, defaultValue=<rgba 0.212 0.212 0.212 1>}
  n10 = ScalarParameter() {parameterName=UseCustomColor}
  n11 = LinearInterpolate(B=Multiply(A=LinearInterpolate(A=@n8, B=@n9.RGB, Alpha=@n10), B=ScalarParameter() {defaultValue=0.5, parameterName=Intensity}), Alpha=@n3) {constAlpha=0.25}
  Multiply(A=VertexColor().A, B=LinearInterpolate(B=Multiply(A=@n6, B=@n11), Alpha=ScalarParameter() {defaultValue=1, parameterName=Anim_Opacity}))
 ]
}
