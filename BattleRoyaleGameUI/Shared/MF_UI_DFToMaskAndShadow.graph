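# UI material function that turns a distance field into two scalar outputs, a Mask and a soft Shadow.
# Mask is a SmoothStep centered on Threshold, with Smoothing divided by the viewport scale factor so the edge width adjusts to UI scale.
# Shadow applies a wide, fixed SmoothStep to the same field, independent of Threshold and Smoothing.
# A remapped circle SDF serves as the preview for the DistanceField input.
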MaterialFunction /BattleRoyaleGameUI/Shared/MF_UI_DFToMaskAndShadow
graph {
 properties {previewMaterialDomain=MD_UI}
 nodes [
  n1 = FunctionInput() {bUsePreviewValueAsDefault=true, inputName=Threshold, inputType=FunctionInput_Scalar, sortPriority=1, previewValue=<0.5 0 0 1>}
  n2 = FunctionInput() {bUsePreviewValueAsDefault=true, inputName=Smoothing, inputType=FunctionInput_Scalar, sortPriority=2, previewValue=<0.2 0 0 1>}
  n3 = Divide(A=@n2, B=MaterialFunctionCall() {materialFunction=/FortUILibrary/MaterialFunctions/Basics/MF_UI_ViewportScale}.Viewport Scale Factor)
  n4 = FunctionInput(Preview=ConstantBiasScale(None=MaterialFunctionCall!(UVs, Radius) {materialFunction=/FortUILibrary/MaterialFunctions/Shapes/SDFs/MF_SDF_Circle}.SDF) {bias=-0.5, scale=-2}) {inputName=DistanceField, inputType=FunctionInput_Scalar}
  FunctionOutput(None=SmoothStep(Min=Subtract(A=@n1, B=@n3), Max=Add(A=@n1, B=@n3), Value=@n4)) {outputName=Mask}
  FunctionOutput(None=SmoothStep(Value=@n4) {constMax=1.2, constMin=0.2}) {outputName=Shadow, sortPriority=1}
 ]
}
