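# Translucent UI master material for flavor text fonts in Battle Royale, with adaptive color and an animated swipe.
# Color blends the Dim adaptive color with itself along an eased vertical gradient from UV1.
# Opacity math eases between Min_Alpha and Max_Alpha on that gradient, with a slightly reduced minimum through a low-end value switcher.
# A sine of rotated UV0, offset by AnimSwipe, forms the SwipeMask that blends this alpha toward full.
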Material /BattleRoyaleGameUI/Shared/M_UI_Font_FlavorText_Master
graph {
 properties {bAutomaticallySetUsageInEditor=false, blendMode=BLEND_Translucent, materialDomain=MD_UI, usedShadingModels=Default Lit}
 nodes [
  n1 = MaterialFunctionCall!(BufferSource) {materialFunction=&/MF_UI_AdaptiveColors}
  n2 = TextureCoordinate() {coordinateIndex=1}
  n3 = MaterialFunctionCall("Value (0-1)"=OneMinus(None=ComponentMask(None=@n2) {g=true})) {materialFunction=&MF_UI_EaseCurves}
  LinearInterpolate(A=@n1.Dim, B=@n1.Dim, Alpha=@n3.EaseOutSine)
  n4 = ScalarParameter() {defaultValue=1, parameterName=Min_Alpha}
  n5 = Multiply(A=@n4) {constB=0.9}
  n6 = TextureCoordinate()
  n7 = ScalarParameter() {defaultValue=-0.125, parameterName=SwipeRotaiton}
  n8 = ScalarParameter() {parameterName=AnimSwipe}
  n9 = Sine(None=Add(
    A=MaterialFunctionCall(UVs=@n6, "Rotation Angle (0-1)"=@n7) {materialFunction=&CustomRotator}.Rotated Values
    B=LinearInterpolate(Alpha=@n8) {constA=-1}
   )) {period=2}
  LinearInterpolate(
   A=LinearInterpolate(
    A=MaterialFunctionCall!(Default=@n4, LowEnd=@n5) {materialFunction=&/MF_UI_LowEnd_ValueSwitcher}.Result
    B=ScalarParameter() {defaultValue=1, parameterName=Max_Alpha}
    Alpha=@n3.EaseInSine
   )
   Alpha=NamedRerouteDeclaration(None=ComponentMask(None=@n9) {r=true}) {name=SwipeMask, nodeColor=<rgba 0 0 0 1>}
  )
  MaterialFunctionCall() {materialFunction=&GetUserInterfaceUV}
 ]
}
