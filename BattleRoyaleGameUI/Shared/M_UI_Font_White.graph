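# Translucent UI font material for the battle royale game UI, tinted from a shared palette in MPC_BRGameUI and white at the default index.
# Index_MPC_Color drives chained lerps through White, Cyan, Green, Orange, Red, StormPink and Yellow.
# Offset, saturated copies of the index gate each step after the first, whose alpha is unclamped, so fractional values blend neighboring colors.
# The result feeds a FinalColor reroute, alongside a near-opaque Opacity parameter.
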Material /BattleRoyaleGameUI/Shared/M_UI_Font_White
graph {
 properties {bAutomaticallySetUsageInEditor=false, blendMode=BLEND_Translucent, materialDomain=MD_UI, usedShadingModels=Default Lit}
 nodes [
  n1 = LinearInterpolate(
   A=ComponentMask(None=CollectionParameter() {collection=&/MPC_BRGameUI, parameterName=White}) {b=true, g=true, r=true}
   B=ComponentMask(None=CollectionParameter() {collection=&MPC_BRGameUI, parameterName=Cyan}) {b=true, g=true, r=true}
   Alpha=ScalarParameter() {parameterName=Index_MPC_Color}
  )
  n2 = ComponentMask(None=CollectionParameter() {collection=&MPC_BRGameUI, parameterName=Green}) {b=true, g=true, r=true}
  n3 = Saturate(None=Subtract(A=ScalarParameter() {parameterName=Index_MPC_Color}))
  n4 = CollectionParameter() {collection=&MPC_BRGameUI, parameterName=Orange}
  n5 = Subtract(A=ScalarParameter() {parameterName=Index_MPC_Color}) {constB=2}
  n6 = ScalarParameter() {parameterName=Index_MPC_Color}
  n7 = LinearInterpolate(
   A=LinearInterpolate(A=LinearInterpolate(A=@n1, B=@n2, Alpha=@n3), B=ComponentMask(None=@n4) {b=true, g=true, r=true}, Alpha=Saturate(None=@n5))
   B=ComponentMask(None=CollectionParameter() {collection=&MPC_BRGameUI, parameterName=Red}) {b=true, g=true, r=true}
   Alpha=Saturate(None=Subtract(A=@n6) {constB=3})
  )
  n8 = ComponentMask(None=CollectionParameter() {collection=&MPC_BRGameUI, parameterName=StormPink}) {b=true, g=true, r=true}
  n9 = Saturate(None=Subtract(A=ScalarParameter() {parameterName=Index_MPC_Color}) {constB=4})
  n10 = CollectionParameter() {collection=&MPC_BRGameUI, parameterName=Yellow}
  n11 = Subtract(A=ScalarParameter() {parameterName=Index_MPC_Color}) {constB=5}
  NamedRerouteDeclaration(None=LinearInterpolate(A=LinearInterpolate(A=@n7, B=@n8, Alpha=@n9), B=ComponentMask(None=@n10) {b=true, g=true, r=true}, Alpha=Saturate(None=@n11))) {name=FinalColor, nodeColor=<rgba 0 0 0 1>}
  NamedRerouteUsage()
  ScalarParameter() {defaultValue=0.95, parameterName=Opacity}
 ]
}
